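# Unlit translucent Niagara sprite material that draws soft round dots for the Season 31 umbrella effects.
# Particle color RGB feeds an emissive adjustment function.
# A centered UV sphere mask, with hardness driven per particle by a dynamic parameter, is raised to a power to shape the dot.
# That mask is multiplied by particle alpha and a parameterized camera depth fade to form the opacity.
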Material /BRCosmetics/Gadgets/Assets/VinderTech_Umbrella/Umbrella_Season31/FX/M_FX_Umbrealla_Season_31_Dots
graph {
 properties {
  bAutomaticallySetUsageInEditor=false
  bUsedWithNiagaraSprites=true
  blendMode=BLEND_Translucent
  shadingModel=MSM_Unlit
  usedShadingModels=Unlit
 }
 nodes [
  n1 = ParticleColor()
  MaterialFunctionCall(Emissive=@n1.RGB) {materialFunction=&MF_FX_EmissiveAdjustments}
  n2 = Constant() {r=0.5}
  n3 = SphereMask(A=TextureCoordinate(), B=@n2, Radius=@n2, Hardness=DynamicParameter() {paramNames [Hardness,Param2,Param3,Param4]}.Hardness)
  Multiply(
   A=Multiply(A=@n1.A, B=Power(Base=@n3))
   B=MaterialFunctionCall(
    Fade Length=ScalarParameter() {defaultValue=100, group=01-Opacity, parameterName=Camera Fade Length, sliderMax=200}
    Fade Offset=ScalarParameter() {defaultValue=25, group=01-Opacity, parameterName=Camera Fade Offset, sliderMax=200}
   ) {materialFunction=&CameraDepthFade}.Result
  )
 ]
}
