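# Translucent, unlit, two-sided Niagara material for the Season 31 umbrella trail effect, a panning flame streak.
# Cell noise distorts a channel-packed flame texture, with UV-gradient shape masks, a glow mask and panned noise that helps blend three colors.
# Opacity adds depth, camera and optional Fresnel fades.
# Particle color tints it, while dynamic parameters offset UVs per particle and drive a smoothstep erosion reveal.
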Material /BRCosmetics/Gadgets/Assets/VinderTech_Umbrella/Umbrella_Season31/FX/M_FX_Umbrealla_Season_31_Trail
graph {
 properties {
  bUsedWithInstancedStaticMeshes=true
  bUsedWithNiagaraMeshParticles=true
  bUsedWithNiagaraRibbons=true
  bUsedWithNiagaraSprites=true
  blendMode=BLEND_Translucent
  pixelDepthOffsetMode=PDOM_Legacy
  shadingModel=MSM_Unlit
  twoSided=true
  usedShadingModels=Unlit
 }
 nodes [
  n1 = DynamicParameter() {defaultValue=<rgba 0.5 0 0 1>, paramNames [UV1 U Offset, Dive, Camera Offset, Reveal]}
  Add(A=@n1.UV1 U Offset)
  n2 = Power(
   Base=MaterialFunctionCall(Float2=TextureCoordinate()) {materialFunction=&BreakOutFloat2Components}.R
   Exp=Constant() {r=1.3}
  )
  n3 = ScalarParameter() {defaultValue=10}
  NamedRerouteDeclaration(None=Saturate(None=Multiply(A=@n2, B=@n3))) {name=BottomFade, nodeColor=<rgba 1 0.8906 0 1>}
  n4 = NamedRerouteUsage()
  n5 = NamedRerouteUsage()
  n6 = Multiply(
   A=SmoothStep(
    Min=ScalarParameter() {group=00- FlameShape, parameterName=Opacity_TextureMin}
    Max=ScalarParameter() {defaultValue=0.4, group=00- FlameShape, parameterName=Opacity_TextureMax}
    Value=Multiply(A=@n4, B=@n5)
   ) {constMax=0.4}
   B=ScalarParameter() {defaultValue=1, group=01-Opacity, parameterName=MainTrailOpacity}
  )
  n7 = Power(
   Base=Multiply(A=NamedRerouteUsage(), B=NamedRerouteUsage())
   Exp=ScalarParameter() {defaultValue=0.5, group=01-Opacity, parameterName=GlowPow}
  )
  n8 = ScalarParameter() {defaultValue=3, group=01-Opacity, parameterName=GlowStrength}
  n9 = Add(A=@n6, B=NamedRerouteDeclaration(None=Multiply(A=@n7, B=@n8)) {name=GlowMask, nodeColor=<rgba 1 0.5859 0 1>})
  n10 = ScalarParameter() {defaultValue=10, group=01-Opacity, parameterName=Depth Fade Distance, sliderMax=100}
  n11 = DepthFade(Opacity=@n9, FadeDistance=LinearInterpolate(A=@n10, B=Multiply(A=@n10) {constB=0.5}, Alpha=@n6)) {fadeDistanceDefault=25}
  n12 = ScalarParameter() {defaultValue=100, group=01-Opacity, parameterName=Camera Fade Length, sliderMax=200}
  n13 = ScalarParameter() {defaultValue=25, group=01-Opacity, parameterName=Camera Fade Offset, sliderMax=200}
  n14 = Fresnel(
   ExponentIn=ScalarParameter() {defaultValue=0.4, parameterName=FresnelExponent}
   BaseReflectFractionIn=ScalarParameter() {defaultValue=0.04, parameterName=FresnelFraction}
  )
  n15 = Multiply(
   A=Multiply(
    A=ShadingPathSwitch(Default=@n11, Deferred=@n11, Forward=@n9, Mobile=@n9)
    B=MaterialFunctionCall(Fade Length=@n12, Fade Offset=@n13) {materialFunction=&CameraDepthFade}.Result
   )
   B=LinearInterpolate(B=OneMinus(None=@n14), Alpha=ScalarParameter() {parameterName=UseFresnel}) {constA=1}
  )
  n16 = ParticleColor()
  n17 = NamedRerouteUsage()
  n18 = Constant() {r=3}
  n19 = Multiply(
   A=Multiply(A=Multiply(A=@n15, B=@n16.A), B=NamedRerouteUsage())
   B=LinearInterpolate(A=Constant() {r=0.2}, B=Constant() {r=1}, Alpha=Power(Base=@n17, Exp=@n18))
  )
  n20 = NamedRerouteUsage()
  NamedRerouteDeclaration(None=Saturate(None=Multiply(A=@n19, B=@n20))) {name=Opacity, nodeColor=<rgba 0 1 0.8359 1>}
  n21 = NamedRerouteUsage()
  n22 = MaterialFunctionCall(Float2=TextureCoordinate()) {materialFunction=&BreakOutFloat2Components}
  n23 = ScalarParameter() {defaultValue=0.2, group=02- Reveal, parameterName=Reveal Softness}
  n24 = DynamicParameter() {defaultValue=<rgba 0.5 0 0 1>, paramNames [UV1 U Offset, Dive, Camera Offset, Reveal]}
  NamedRerouteDeclaration(None=MaterialFunctionCall!(
    Texture=OneMinus(None=@n21)
    Erosion=MaterialFunctionCall(Texture=@n22.R, Erosion=@n24.Reveal, Softness=@n23) {materialFunction=/Game/Effects/Niagara/MasterMaterials/Functions/MF_Erosion_Smoothstep_01}.Result
    Softness=Constant() {r=0.5}
   ) {materialFunction=&MF_Erosion_Smoothstep_01}.Result) {name=RevealMask, nodeColor=<rgba 0.125 0 1 1>}
  n25 = VectorParameter() {group=03- Color, parameterName=ColorTex_Settings, channelNames {a=PanY, b=PanX, g=ScaleY, r=ScaleX}, defaultValue=<rgba 1 1 -0.4 0>}
  n26 = Multiply(
   A=Panner(Coordinate=TextureCoordinate(), Speed=AppendVector(A=@n25.PanX, B=@n25.PanX)) {speedX=-0.5}
   B=AppendVector(A=@n25.ScaleX, B=@n25.ScaleX)
  )
  n27 = AppendVector(A=@n1.UV1 U Offset, B=Constant())
  n28 = NamedRerouteUsage()
  n29 = ScalarParameter() {defaultValue=0.2, group=03- Color, parameterName=ColorDistoStrength}
  n30 = TextureSampleParameter2D(UVs=Add(A=Add(A=@n26, B=@n27), B=Multiply(A=@n28, B=@n29))) {
   group=03- Color
   parameterName=ColorTex
   samplerType=SAMPLERTYPE_LinearColor
   texture=/Game/Effects/Fort_Effects/Gameplay/SmokeDustExplosions/Textures/T_NoisePack_LinearColor
  }
  NamedRerouteDeclaration(None=@n30.G) {name=ColorNoiseGreen, nodeColor=<rgba 0 1 0.6719 1>}
  n31 = Multiply(A=DynamicParameter() {defaultValue=<rgba 0.5 0 0 1>, paramNames [UV1 U Offset, Dive, Camera Offset, Reveal]}.UV1 U Offset) {constB=3}
  n32 = Constant()
  n33 = TextureCoordinate()
  n34 = VectorParameter() {group=06- Distortion, parameterName=DistoUV_Settings, channelNames {a=PanY, b=PanX, g=ScaleY, r=ScaleX}, defaultValue=<rgba 2 1.3 -0.65 0>}
  n35 = Panner(Coordinate=@n33, Speed=AppendVector(A=@n34.PanX, B=@n34.PanX)) {speedX=-0.5}
  n36 = AppendVector(A=@n34.ScaleX, B=@n34.ScaleX)
  n37 = TextureSampleParameter2D(UVs=Add(A=AppendVector(A=@n31, B=@n32), B=Multiply(A=@n35, B=@n36))) {
   group=06- Distortion
   parameterName=Distortion Texture3
   samplerType=SAMPLERTYPE_LinearGrayscale
   texture=/Game/Effects/Fort_Effects/Textures/Tile/T_CellNoise_Tile_M
  }
  n38 = Add(A=OneMinus(None=ComponentMask(None=@n33) {r=true})) {constB=-0.5}
  NamedRerouteDeclaration(None=Multiply(A=@n37.R, B=Multiply(A=@n38) {constB=-1})) {name=Distortion, nodeColor=<rgba 1 0.3083 0.7353 1>}
  n39 = NamedRerouteUsage()
  n40 = ScalarParameter() {defaultValue=0.05, group=06- Distortion, parameterName=Distortion Strength, sliderMax=1, sliderMin=-1}
  n41 = Panner(
   Coordinate=Add(A=TextureCoordinate(), B=Multiply(A=@n39, B=@n40) {constB=0.1})
   Speed=AppendVector(A=ScalarParameter() {defaultValue=-0.6, group=00- FlameShape, parameterName=TrailSpeedX, sliderMax=5, sliderMin=-5}, B=Constant())
  ) {speedX=1}
  n42 = Constant() {r=1}
  n43 = AppendVector(A=ScalarParameter() {defaultValue=1.5, group=00- FlameShape, parameterName=Scale X, sliderMax=5}, B=@n42)
  n44 = ScalarParameter() {defaultValue=2, group=00- FlameShape, parameterName=Scale Y, sliderMax=5}
  n45 = MaterialFunctionCall(Float2=MaterialFunctionCall(
    UVs=Multiply(A=@n41, B=@n43)
    Texture Scale=MaterialFunctionCall(X=@n42, Y=@n44) {materialFunction=&MakeFloat2}.Result
   ) {materialFunction=&ScaleUVsByCenter}.UVs) {materialFunction=&BreakOutFloat2Components}
  n46 = DynamicParameter() {defaultValue=<rgba 0.5 0 0 1>, paramNames [UV1 U Offset, Dive, Camera Offset, Reveal]}
  n47 = Constant() {r=0.5}
  n48 = Add(
   A=MaterialFunctionCall(X=@n45.R, Y=Saturate(None=@n45.G)) {materialFunction=&MakeFloat2}.Result
   B=AppendVector(A=Multiply(A=@n46.UV1 U Offset, B=@n47), B=Constant())
  )
  n49 = TextureSample(UVs=LinearInterpolate(
    A=@n48
    B=MaterialFunctionCall(XY=@n48) {materialFunction=&Swizzle}.YX
    Alpha=ScalarParameter() {group=00- FlameShape, parameterName=SwizzleTrail}
   )) {texture=/BRCosmetics/Characters/Player/Female/Medium/Bodies/F_MED_MadameMoth_Posh/FX/FX_Rig_Pipeline/T_Rig_MadameMoth_Fire_Trail01}
  NamedRerouteDeclaration(None=@n49.R) {name=FlameTrail, nodeColor=<rgba 0 1 0.2734 1>}
  NamedRerouteDeclaration(None=@n37.R) {name=DistortionTexture, nodeColor=<rgba 0.125 0 1 1>}
  NamedRerouteDeclaration(None=@n49.G) {name=FlameTrailGlow, nodeColor=<rgba 1 0 0.8906 1>}
  n50 = TextureCoordinate()
  n51 = ComponentMask(None=@n50) {r=true}
  n52 = SmoothStep(Max=ScalarParameter() {group=00- FlameShape, parameterName=FlameBaseSoftness}, Value=@n51) {constMax=0.2}
  n53 = SmoothStep(Max=ScalarParameter() {defaultValue=0.8, group=00- FlameShape, parameterName=FlameTipSoftness}, Value=OneMinus(None=@n51)) {constMax=0.1}
  n54 = ScalarParameter() {defaultValue=0.43, group=00- FlameShape, parameterName=FlameSideSoftness}
  n55 = ComponentMask(None=@n50) {g=true}
  n56 = Multiply(A=@n55, B=OneMinus(None=@n55))
  n57 = Subtract(
   A=Multiply(A=Multiply(A=@n52, B=@n53), B=SmoothStep(Max=@n54, Value=@n56) {constMax=0.5})
   B=Multiply(A=NamedRerouteUsage(), B=ScalarParameter() {defaultValue=0.55, group=00- FlameShape, parameterName=SubstractDisto}) {constB=0.25}
  )
  NamedRerouteDeclaration(None=Multiply(A=ComponentMask(None=@n57) {r=true}, B=Constant() {r=2}) {constB=3}) {name=MainShape, nodeColor=<rgba 0.2188 0 1 1>}
  n58 = ParticleColor()
  n59 = VectorParameter() {group=03- Color, parameterName=ColorA, sortPriority=0, defaultValue=<rgba 0.007 0.7 0.0994 1>}
  n60 = VectorParameter() {group=03- Color, parameterName=ColorB, sortPriority=0, defaultValue=<rgba 0.3308 0.92 0 1>}
  n61 = OneMinus(None=ComponentMask(None=TextureCoordinate()) {r=true})
  n62 = ScalarParameter() {defaultValue=8, parameterName=VerticalColorMaskPower}
  n63 = Saturate(None=Multiply(A=Power(Base=@n61, Exp=@n62), B=ScalarParameter() {defaultValue=3, parameterName=VerticalColorMaskMult}))
  n64 = ScalarParameter() {defaultValue=0.2, group=03- Color, parameterName=FirstColorThreshold}
  n65 = SmoothStep(
   Min=@n64
   Max=Add(A=@n64, B=ScalarParameter() {defaultValue=1, group=03- Color, parameterName=FirstColorSoftness})
   Value=Saturate(None=NamedRerouteUsage())
  ) {constMin=0.5}
  n66 = Saturate(None=Multiply(A=Saturate(None=@n65), B=ScalarParameter() {defaultValue=1.3, parameterName=ColorNoiseAmount}))
  n67 = LinearInterpolate(B=ScalarParameter() {defaultValue=1, parameterName=SecColorAmount}, Alpha=Saturate(None=Add(A=@n63, B=@n66) {constB=0}))
  n68 = Power(Base=OneMinus(None=NamedRerouteUsage()), Exp=ScalarParameter() {defaultValue=1, parameterName=GlowMaskPower})
  n69 = Multiply(
   A=LinearInterpolate(
    A=LinearInterpolate(A=@n59.RGB, B=@n60.RGB, Alpha=@n67)
    B=VectorParameter() {group=03- Color, parameterName=ColorC, defaultValue=<rgba 0 0 0 1>}.RGB
    Alpha=Saturate(None=@n68)
   )
   B=ScalarParameter() {defaultValue=1, group=03- Color, parameterName=Emissive, sliderMax=2}
  )
  NamedRerouteDeclaration(None=MaterialFunctionCall(Emissive=Multiply(A=@n58.RGB, B=@n69) {constA=1}) {materialFunction=&MF_FX_EmissiveAdjustments}.Result) {name=Emissive, nodeColor=<rgba 0.2656 0 1 1>}
  NamedRerouteDeclaration(None=@n30.R) {name=ColorNoise, nodeColor=<rgba 0.3359 0 1 1>}
  NamedRerouteDeclaration(None=@n49.B) {name=FlameTrailInnerColor, nodeColor=<rgba 0 1 0.25 1>}
  NamedRerouteUsage()
  NamedRerouteUsage()
  VertexColor()
 ]
}
