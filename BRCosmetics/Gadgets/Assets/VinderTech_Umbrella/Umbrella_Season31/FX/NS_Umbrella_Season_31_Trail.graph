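# Niagara system for the Season 31 umbrella cosmetic's trail: two spiraling ribbons, trailing dot sprites and a top glow.
# Trail_L/R rebuild burst-spawned ribbon points each update from spirals, slerped owner axes and bends driven by smoothed User.RightDot/ForwardDot.
# Dots_L/R read those ribbons to follow them, with movement-scaled spawn rate and lifetime.
# System RevealCurve drives dot size, glow alpha and ribbon material; front-end modifiers adjust placement and glow.
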NiagaraSystem /BRCosmetics/Gadgets/Assets/VinderTech_Umbrella/Umbrella_Season31/FX/NS_Umbrella_Season_31_Trail
dependencies {
 usedDataInterfaces [
  &NiagaraDataInterfaceCurve
  &NiagaraDataInterfaceParticleRead
  &NiagaraDataInterfaceCamera
  &NiagaraDataInterfaceSkeletalMesh
 ]
 usedDynamicInputs [
  &MergeHelper_ColorSwitch
  /Game/Effects/Niagara/DynamicInputs/Math/FN_RemapDotProduct
  &ReturnCameraVector
 ]
 usedModules [
  &SystemState
  &FN_Integrator
  &self:SystemSpawnScript.SystemScriptSource.SystemScriptGraph.NiagaraNodeAssignment_0.SetVariables_#1
  &self:SystemSpawnScript.SystemScriptSource.SystemScriptGraph.NiagaraNodeAssignment_1.SetVariables_#2
  &self:SystemSpawnScript.SystemScriptSource.SystemScriptGraph.NiagaraNodeAssignment_3.SetVariables_#3
  &self:SystemSpawnScript.SystemScriptSource.SystemScriptGraph.NiagaraNodeAssignment_2.SetVariables_#4
 ]
}
emitters {
 Dots_L {
  emitterUpdateScript [
   &EmitterState(Inactive Response=::NewEnumerator1, Life Cycle Mode=::NewEnumerator1, Loop Delay=0.2, Loop Duration=1, UseLoopDelay=true)
   &SpawnRate(
    SpawnRate=&Multiply_Float(
     A=6
     B=&Lerp_Float(
      A=1
      B=2
      Alpha=&ClampFloat(Float=&Add_Float(
        A=&Abs_Float(Float:NiagaraFloat=$User.RightDot)
        B=&Abs_Float(Float:NiagaraFloat=$User.ForwardDot)
       ))
     )
    )
    Use Spawn Probability=false
   )
  ]
  particleSpawnScript [
   &FN_SetSpawnGroup(Number of Spawn Groups=8)
   &InitializeParticle(
    Color=$System.ColorDots
    Lifetime=&Multiply_Float(
     A=&RandomRangeFloat(Minimum=2.5, Maximum=4)
     B=&Lerp_Float(
      A=1
      B=0.6
      Alpha=&ClampFloat(Float=&Add_Float(
        A=&Abs_Float(Float:NiagaraFloat=$User.RightDot)
        B=&Abs_Float(Float:NiagaraFloat=$User.ForwardDot)
       ))
     )
    )
    Sprite Size Mode=::NewEnumerator3
    Uniform Sprite Size=1
    UsePositionOffset=false
    Write Material Random=true
   ) {module=InitializeParticle001}
   /SetVariables(Particles.SpriteSize=&Vector2DFromFloat(Value=&UniformAOrBFloat(
      A=5.5
      B=&RandomRangeFloat(Minimum=3.8, Maximum=4.2)
      Distribution Weight=0.9
      Evaluation Type=::NewEnumerator0
      Recalculate Random Each Loop=true
      Override Seed=false
     ))) {isSetParameters=true, module=SetVariables_#5}
   &SkeletalMeshLocation(MaskSpawnGroup=false, Mesh Sampling Type=::NewEnumerator5, Skeletal Mesh=$System.Skeletal Mesh, Socket Sampling Mode=::NewEnumerator5)
   &DynamicMaterialParameters(
    Index 0 Param 1=&UniformAOrBFloat(A=0.333, Evaluation Type=::NewEnumerator0, Recalculate Random Each Loop=true, Override Seed=false)
    Module.Index0Param0Write=true
    Module.Index0Param1Write=false
    Module.Index0Param2Write=false
    Module.Index0Param3Write=false
   )
  ]
  particleUpdateScript [
   &ParticleState() {module=ParticleState001}
   &self:FollowRibbon(
    Alpha:NiagaraFloat=&FloatFromCurve(FloatCurve=DataInterface() {propertyValues {Curve {keys [<at 0 1>,<at 1 0>]}}})
    ParticlePerRibbon=14
    RibbonReader=DataInterface() {propertyValues {EmitterBinding {bindingMode=Other, emitterName=Trail_L}}}
   ) {module=FollowRibbon}
   &InterpolateOverTime(Rate Of Change=7, Target Position Value:NiagaraPosition=$Output.FollowRibbon.RibbonPos, Variable Type=::NewEnumerator6)
   /SetVariables(Particles.Position=&AddVectorToPosition(
     Position=&Multiply_Vector(
      A=$Particles.InterpolateOverTime.InterpolateOverTime_Position.Moving Average
      B=&VectorFromFloat(Value=1)
     )
     Vector=&Add_Vector(
      A=&Multiply_Vector(
       A=&RandomRangeVector(Minimum=<-20 -20 -10>, Maximum=<20 20 10>)
       B=&VectorFromFloat(Value=&FloatFromCurve(
         FloatCurve=DataInterface() {propertyValues {Curve {keys [
             {arriveTangent=0.9275, interpMode=RCIM_Cubic, leaveTangent=1.2466, tangentMode=RCTM_Break, time=0, value=0}
             {interpMode=RCIM_Cubic, tangentMode=RCTM_Break, time=1, value=1}
            ]}}}
         Scale Curve=0.5
        ))
      )
      B=<0 0 0>
     )
    )) {isSetParameters=true, module=SetVariables_#6}
   &ParticleState()
   &CameraOffset(Camera Offset Amount=5)
   &ScaleColor(ScaleA=true, ScaleRGB=false)
   /SetVariables(
    Particles.AdditionalDot01Color:LinearColor=$System.ColorDots
    Particles.AdditionalDot01Position:NiagaraPosition=&AddVectorToPosition(
     Position=$Particles.Position
     Vector=&Multiply_VectorByFloat(
      Vector=&RandomRangeVector(Minimum=<-1 -1 -1>, Maximum=<1 1 1>)
      Float=&Waveform(
       Global Input Position=&RandomRangeFloat(Minimum=0, Maximum=1)
       "Waveform [1]"=::NewEnumerator0
       "[1] Frequency"=1
       "[1] Amplitude Min/Max"=<1.5 2.5>
      )
     )
    )
    Particles.AdditionalDot01SpriteSize:Vector2f=&Multiply_Vector2D_ByFloat(
     Float=&Multiply_Float(
      A=&UniformAOrBFloat(
       A=&FloatFromCurve(FloatCurve=DataInterface() {propertyValues {Curve {keys [
            <at 0 0 RCIM_Cubic>
            {arriveTangent=9.6335, interpMode=RCIM_Cubic, leaveTangent=2.7827, tangentMode=RCTM_Break, time=0.15, value=1}
            <at 0.6943 0.7273 RCIM_Cubic>
            {arriveTangent=2.7202, interpMode=RCIM_Cubic, leaveTangent=-2.3793, tangentMode=RCTM_Break, time=0.8055, value=1}
            <at 0.95 0 RCIM_Cubic>
           ]}}})
       B=&FloatFromCurve(FloatCurve=DataInterface() {propertyValues {Curve {keys [
            <at 0 0 RCIM_Cubic>
            {arriveTangent=9.6335, interpMode=RCIM_Cubic, leaveTangent=2.7827, tangentMode=RCTM_Break, time=0.15, value=0.9794}
            {arriveTangent=0.0322, interpMode=RCIM_Cubic, leaveTangent=0.0322, time=0.5397, value=0.5}
            {arriveTangent=2.7202, interpMode=RCIM_Cubic, leaveTangent=-2.3793, tangentMode=RCTM_Break, time=0.7888, value=1}
            <at 0.95 0 RCIM_Cubic>
           ]}}})
       Evaluation Type=::NewEnumerator0
       Recalculate Random Each Loop=true
       Override Seed=false
      )
      B=$System.RevealCurve
     )
     Vector2D=&Vector2DFromFloat(Value=&UniformAOrBFloat(
       A=5.1
       B=&RandomRangeFloat(Minimum=3.8, Maximum=4.4)
       Distribution Weight=0.9
       Evaluation Type=::NewEnumerator0
       Recalculate Random Each Loop=true
       Override Seed=false
      ))
    )
    Particles.AdditionalDot01Visibility:NiagaraInt32=&MakeIntFromFloat(Float=&UniformAOrBFloat(A=0, Distribution Weight=0.6, Evaluation Type=::NewEnumerator0, Recalculate Random Each Loop=true, Override Seed=false))
   ) {isSetParameters=true, module=SetVariables_#7}
   /SetVariables(
    Particles.AdditionalDot02Color:LinearColor=$System.ColorDots
    Particles.AdditionalDot02Position:NiagaraPosition=&AddVectorToPosition(
     Position=$Particles.Position
     Vector=&Multiply_VectorByFloat(
      Vector=&RandomRangeVector(Minimum=<-1 -1 -1>, Maximum=<1 1 1>)
      Float=&Waveform(
       Global Input Position=&RandomRangeFloat(Minimum=0, Maximum=1)
       "Waveform [1]"=::NewEnumerator0
       "[1] Frequency"=1
       "[1] Amplitude Min/Max"=<1.5 2.5>
      )
     )
    )
    Particles.AdditionalDot02SpriteSize:Vector2f=&Multiply_Vector2D_ByFloat(
     Float=&Multiply_Float(
      A=&UniformAOrBFloat(
       A=&FloatFromCurve(FloatCurve=DataInterface() {propertyValues {Curve {keys [
            <at 0 0 RCIM_Cubic>
            {arriveTangent=9.6335, interpMode=RCIM_Cubic, leaveTangent=2.7827, tangentMode=RCTM_Break, time=0.15, value=1}
            <at 0.6943 0.7273 RCIM_Cubic>
            {arriveTangent=2.7202, interpMode=RCIM_Cubic, leaveTangent=-2.3793, tangentMode=RCTM_Break, time=0.8055, value=1}
            <at 0.95 0 RCIM_Cubic>
           ]}}})
       B=&FloatFromCurve(FloatCurve=DataInterface() {propertyValues {Curve {keys [
            <at 0 0 RCIM_Cubic>
            {arriveTangent=9.6335, interpMode=RCIM_Cubic, leaveTangent=2.7827, tangentMode=RCTM_Break, time=0.15, value=0.9794}
            {arriveTangent=0.0322, interpMode=RCIM_Cubic, leaveTangent=0.0322, time=0.5397, value=0.5}
            {arriveTangent=2.7202, interpMode=RCIM_Cubic, leaveTangent=-2.3793, tangentMode=RCTM_Break, time=0.7888, value=1}
            <at 0.95 0 RCIM_Cubic>
           ]}}})
       Evaluation Type=::NewEnumerator0
       Recalculate Random Each Loop=true
       Override Seed=false
      )
      B=$System.RevealCurve
     )
     Vector2D=&Vector2DFromFloat(Value=&UniformAOrBFloat(
       A=5.1
       B=&RandomRangeFloat(Minimum=3.8, Maximum=4.4)
       Distribution Weight=0.9
       Evaluation Type=::NewEnumerator0
       Recalculate Random Each Loop=true
       Override Seed=false
      ))
    )
    Particles.AdditionalDot02Visibility:NiagaraInt32=&MakeIntFromFloat(Float=&UniformAOrBFloat(A=0, Distribution Weight=0.88, Evaluation Type=::NewEnumerator0, Recalculate Random Each Loop=true, Override Seed=false))
   ) {isSetParameters=true, module=SetVariables_#8}
   /SetVariables(Particles.SpriteSize=&Multiply_Vector2D_ByFloat(
     Float=&Multiply_Float(
      A=$System.RevealCurve
      B=&UniformAOrBFloat(
       A=&FloatFromCurve(FloatCurve=DataInterface() {propertyValues {Curve {keys [
            <at 0 0 RCIM_Cubic>
            {arriveTangent=9.6335, interpMode=RCIM_Cubic, leaveTangent=2.7827, tangentMode=RCTM_Break, time=0.15, value=1}
            {arriveTangent=1.2288, interpMode=RCIM_Cubic, leaveTangent=1.2288, time=0.3322, value=0.4478}
            {arriveTangent=4.272, interpMode=RCIM_Cubic, leaveTangent=0.772, tangentMode=RCTM_Break, time=0.5569, value=1.5}
            {arriveTangent=-2.0113, interpMode=RCIM_Cubic, leaveTangent=-2.0113, time=0.6943, value=0.7273}
            {arriveTangent=2.7202, interpMode=RCIM_Cubic, leaveTangent=-2.3793, tangentMode=RCTM_Break, time=0.8055, value=1}
            <at 0.95 0 RCIM_Cubic>
           ]}}})
       B=&FloatFromCurve(FloatCurve=DataInterface() {propertyValues {Curve {keys [
            <at 0 0 RCIM_Cubic>
            {arriveTangent=9.6335, interpMode=RCIM_Cubic, leaveTangent=2.7827, tangentMode=RCTM_Break, time=0.15, value=0.9794}
            {arriveTangent=0.0322, interpMode=RCIM_Cubic, leaveTangent=0.0322, time=0.5397, value=0.5}
            {arriveTangent=2.7202, interpMode=RCIM_Cubic, leaveTangent=-2.3793, tangentMode=RCTM_Break, time=0.7888, value=1}
            <at 0.95 0 RCIM_Cubic>
           ]}}})
       Evaluation Type=::NewEnumerator0
       Recalculate Random Each Loop=true
       Override Seed=false
      )
     )
     Vector2D=$Particles.Initial.SpriteSize
    )) {isSetParameters=true, module=SetVariables_#9}
  ]
  rendererClasses [&NiagaraSpriteRendererProperties]
  renderers [
   {rendererClass=&NiagaraSpriteRendererProperties}
   {rendererClass=&NiagaraSpriteRendererProperties, rendererIndex=1}
   {rendererClass=&NiagaraSpriteRendererProperties, rendererIndex=2}
  ]
 }
 Dots_R {
  emitterUpdateScript [
   &EmitterState(Inactive Response=::NewEnumerator1, Life Cycle Mode=::NewEnumerator1, Loop Delay=0.2, Loop Duration=1, UseLoopDelay=true)
   &SpawnRate(
    SpawnRate=&Multiply_Float(
     A=5
     B=&Lerp_Float(
      A=1
      B=2
      Alpha=&ClampFloat(Float=&Add_Float(
        A=&Abs_Float(Float:NiagaraFloat=$User.RightDot)
        B=&Abs_Float(Float:NiagaraFloat=$User.ForwardDot)
       ))
     )
    )
    Use Spawn Probability=false
   )
  ]
  particleSpawnScript [
   &FN_SetSpawnGroup(Number of Spawn Groups=8)
   &InitializeParticle(
    Color=$System.ColorDots
    Lifetime=&Multiply_Float(
     A=&RandomRangeFloat(Minimum=2.5, Maximum=4)
     B=&Lerp_Float(
      A=1
      B=0.6
      Alpha=&ClampFloat(Float=&Add_Float(
        A=&Abs_Float(Float:NiagaraFloat=$User.RightDot)
        B=&Abs_Float(Float:NiagaraFloat=$User.ForwardDot)
       ))
     )
    )
    Sprite Size Mode=::NewEnumerator3
    Uniform Sprite Size=1
    UsePositionOffset=false
    Write Material Random=true
   ) {module=InitializeParticle001}
   /SetVariables(Particles.SpriteSize=&Vector2DFromFloat(Value=&UniformAOrBFloat(
      A=5.5
      B=&RandomRangeFloat(Minimum=3.8, Maximum=4.2)
      Distribution Weight=0.9
      Evaluation Type=::NewEnumerator0
      Recalculate Random Each Loop=true
      Override Seed=false
     ))) {isSetParameters=true, module=SetVariables_#5}
   &SkeletalMeshLocation(MaskSpawnGroup=false, Mesh Sampling Type=::NewEnumerator5, Skeletal Mesh=$System.Skeletal Mesh, Socket Sampling Mode=::NewEnumerator5)
   &DynamicMaterialParameters(
    Index 0 Param 1=&UniformAOrBFloat(A=0.333, Evaluation Type=::NewEnumerator0, Recalculate Random Each Loop=true, Override Seed=false)
    Module.Index0Param0Write=true
    Module.Index0Param1Write=false
    Module.Index0Param2Write=false
    Module.Index0Param3Write=false
   )
  ]
  particleUpdateScript [
   &ParticleState() {module=ParticleState001}
   &self:FollowRibbon(
    Alpha:NiagaraFloat=&FloatFromCurve(FloatCurve=DataInterface() {propertyValues {Curve {keys [<at 0 1>,<at 1 0>]}}})
    ParticlePerRibbon=10
    RibbonReader=DataInterface() {propertyValues {EmitterBinding {bindingMode=Other, emitterName=Trail_R}}}
   ) {module=FollowRibbon}
   &InterpolateOverTime(Rate Of Change=7, Target Position Value:NiagaraPosition=$Output.FollowRibbon.RibbonPos, Variable Type=::NewEnumerator6)
   /SetVariables(Particles.Position=&AddVectorToPosition(
     Position=&Multiply_Vector(
      A=$Particles.InterpolateOverTime.InterpolateOverTime_Position.Moving Average
      B=&VectorFromFloat(Value=1)
     )
     Vector=&Add_Vector(
      A=&Multiply_Vector(
       A=&RandomRangeVector(Minimum=<-20 -20 -10>, Maximum=<20 20 10>)
       B=&VectorFromFloat(Value=&FloatFromCurve(
         FloatCurve=DataInterface() {propertyValues {Curve {keys [
             {arriveTangent=0.9275, interpMode=RCIM_Cubic, leaveTangent=1.2466, tangentMode=RCTM_Break, time=0, value=0}
             {interpMode=RCIM_Cubic, tangentMode=RCTM_Break, time=1, value=1}
            ]}}}
         Scale Curve=0.5
        ))
      )
      B=<0 0 0>
     )
    )) {isSetParameters=true, module=SetVariables_#6}
   &ParticleState()
   &CameraOffset(Camera Offset Amount=5)
   &ScaleColor(ScaleA=true, ScaleRGB=false)
   /SetVariables(
    Particles.AdditionalDot01Color:LinearColor=$System.ColorDots
    Particles.AdditionalDot01Position:NiagaraPosition=&AddVectorToPosition(
     Position=$Particles.Position
     Vector=&Multiply_VectorByFloat(
      Vector=&RandomRangeVector(Minimum=<-1 -1 -1>, Maximum=<1 1 1>)
      Float=&Waveform(
       Global Input Position=&RandomRangeFloat(Minimum=0, Maximum=1)
       "Waveform [1]"=::NewEnumerator0
       "[1] Frequency"=1
       "[1] Amplitude Min/Max"=<1.5 2.5>
      )
     )
    )
    Particles.AdditionalDot01SpriteSize:Vector2f=&Multiply_Vector2D_ByFloat(
     Float=&Multiply_Float(
      A=&UniformAOrBFloat(
       A=&FloatFromCurve(FloatCurve=DataInterface() {propertyValues {Curve {keys [
            <at 0 0 RCIM_Cubic>
            {arriveTangent=9.6335, interpMode=RCIM_Cubic, leaveTangent=2.7827, tangentMode=RCTM_Break, time=0.15, value=1}
            <at 0.6943 0.7273 RCIM_Cubic>
            {arriveTangent=2.7202, interpMode=RCIM_Cubic, leaveTangent=-2.3793, tangentMode=RCTM_Break, time=0.8055, value=1}
            <at 0.95 0 RCIM_Cubic>
           ]}}})
       B=&FloatFromCurve(FloatCurve=DataInterface() {propertyValues {Curve {keys [
            <at 0 0 RCIM_Cubic>
            {arriveTangent=9.6335, interpMode=RCIM_Cubic, leaveTangent=2.7827, tangentMode=RCTM_Break, time=0.15, value=0.9794}
            {arriveTangent=0.0322, interpMode=RCIM_Cubic, leaveTangent=0.0322, time=0.5397, value=0.5}
            {arriveTangent=2.7202, interpMode=RCIM_Cubic, leaveTangent=-2.3793, tangentMode=RCTM_Break, time=0.7888, value=1}
            <at 0.95 0 RCIM_Cubic>
           ]}}})
       Evaluation Type=::NewEnumerator0
       Recalculate Random Each Loop=true
       Override Seed=false
      )
      B=$System.RevealCurve
     )
     Vector2D=&Vector2DFromFloat(Value=&UniformAOrBFloat(
       A=5.1
       B=&RandomRangeFloat(Minimum=3.8, Maximum=4.4)
       Distribution Weight=0.9
       Evaluation Type=::NewEnumerator0
       Recalculate Random Each Loop=true
       Override Seed=false
      ))
    )
    Particles.AdditionalDot01Visibility:NiagaraInt32=&MakeIntFromFloat(Float=&UniformAOrBFloat(A=0, Distribution Weight=0.6, Evaluation Type=::NewEnumerator0, Recalculate Random Each Loop=true, Override Seed=false))
   ) {isSetParameters=true, module=SetVariables_#7}
   /SetVariables(
    Particles.AdditionalDot02Color:LinearColor=$System.ColorDots
    Particles.AdditionalDot02Position:NiagaraPosition=&AddVectorToPosition(
     Position=$Particles.Position
     Vector=&Multiply_VectorByFloat(
      Vector=&RandomRangeVector(Minimum=<-1 -1 -1>, Maximum=<1 1 1>)
      Float=&Waveform(
       Global Input Position=&RandomRangeFloat(Minimum=0, Maximum=1)
       "Waveform [1]"=::NewEnumerator0
       "[1] Frequency"=1
       "[1] Amplitude Min/Max"=<1.5 2.5>
      )
     )
    )
    Particles.AdditionalDot02SpriteSize:Vector2f=&Multiply_Vector2D_ByFloat(
     Float=&Multiply_Float(
      A=&UniformAOrBFloat(
       A=&FloatFromCurve(FloatCurve=DataInterface() {propertyValues {Curve {keys [
            <at 0 0 RCIM_Cubic>
            {arriveTangent=9.6335, interpMode=RCIM_Cubic, leaveTangent=2.7827, tangentMode=RCTM_Break, time=0.15, value=1}
            <at 0.6943 0.7273 RCIM_Cubic>
            {arriveTangent=2.7202, interpMode=RCIM_Cubic, leaveTangent=-2.3793, tangentMode=RCTM_Break, time=0.8055, value=1}
            <at 0.95 0 RCIM_Cubic>
           ]}}})
       B=&FloatFromCurve(FloatCurve=DataInterface() {propertyValues {Curve {keys [
            <at 0 0 RCIM_Cubic>
            {arriveTangent=9.6335, interpMode=RCIM_Cubic, leaveTangent=2.7827, tangentMode=RCTM_Break, time=0.15, value=0.9794}
            {arriveTangent=0.0322, interpMode=RCIM_Cubic, leaveTangent=0.0322, time=0.5397, value=0.5}
            {arriveTangent=2.7202, interpMode=RCIM_Cubic, leaveTangent=-2.3793, tangentMode=RCTM_Break, time=0.7888, value=1}
            <at 0.95 0 RCIM_Cubic>
           ]}}})
       Evaluation Type=::NewEnumerator0
       Recalculate Random Each Loop=true
       Override Seed=false
      )
      B=$System.RevealCurve
     )
     Vector2D=&Vector2DFromFloat(Value=&UniformAOrBFloat(
       A=5.1
       B=&RandomRangeFloat(Minimum=3.8, Maximum=4.4)
       Distribution Weight=0.9
       Evaluation Type=::NewEnumerator0
       Recalculate Random Each Loop=true
       Override Seed=false
      ))
    )
    Particles.AdditionalDot02Visibility:NiagaraInt32=&MakeIntFromFloat(Float=&UniformAOrBFloat(A=0, Distribution Weight=0.88, Evaluation Type=::NewEnumerator0, Recalculate Random Each Loop=true, Override Seed=false))
   ) {isSetParameters=true, module=SetVariables_#8}
   /SetVariables(Particles.SpriteSize=&Multiply_Vector2D_ByFloat(
     Float=&Multiply_Float(
      A=$System.RevealCurve
      B=&UniformAOrBFloat(
       A=&FloatFromCurve(FloatCurve=DataInterface() {propertyValues {Curve {keys [
            <at 0 0 RCIM_Cubic>
            {arriveTangent=9.6335, interpMode=RCIM_Cubic, leaveTangent=2.7827, tangentMode=RCTM_Break, time=0.15, value=1}
            {arriveTangent=1.2288, interpMode=RCIM_Cubic, leaveTangent=1.2288, time=0.3322, value=0.4478}
            {arriveTangent=4.272, interpMode=RCIM_Cubic, leaveTangent=0.772, tangentMode=RCTM_Break, time=0.5569, value=1.5}
            {arriveTangent=-2.0113, interpMode=RCIM_Cubic, leaveTangent=-2.0113, time=0.6943, value=0.7273}
            {arriveTangent=2.7202, interpMode=RCIM_Cubic, leaveTangent=-2.3793, tangentMode=RCTM_Break, time=0.8055, value=1}
            <at 0.95 0 RCIM_Cubic>
           ]}}})
       B=&FloatFromCurve(FloatCurve=DataInterface() {propertyValues {Curve {keys [
            <at 0 0 RCIM_Cubic>
            {arriveTangent=9.6335, interpMode=RCIM_Cubic, leaveTangent=2.7827, tangentMode=RCTM_Break, time=0.15, value=0.9794}
            {arriveTangent=0.0322, interpMode=RCIM_Cubic, leaveTangent=0.0322, time=0.5397, value=0.5}
            {arriveTangent=2.7202, interpMode=RCIM_Cubic, leaveTangent=-2.3793, tangentMode=RCTM_Break, time=0.7888, value=1}
            <at 0.95 0 RCIM_Cubic>
           ]}}})
       Evaluation Type=::NewEnumerator0
       Recalculate Random Each Loop=true
       Override Seed=false
      )
     )
     Vector2D=$Particles.Initial.SpriteSize
    )) {isSetParameters=true, module=SetVariables_#9}
  ]
  rendererClasses [&NiagaraSpriteRendererProperties]
  renderers [
   {rendererClass=&NiagaraSpriteRendererProperties}
   {rendererClass=&NiagaraSpriteRendererProperties, rendererIndex=1}
   {rendererClass=&NiagaraSpriteRendererProperties, rendererIndex=2}
  ]
 }
 FN_Debug {
  bEnabled=false
  emitterUpdateScript [
   &EmitterState()
   &SpawnBurst_Instantaneous(Spawn Count=1, Use Spawn Probability=false)
  ]
  particleSpawnScript [/Niagara/Modules/Spawn/Initialization/InitializeParticle(
    Lifetime=5
    Module.Write Color=true
    Module.Write Lifetime=true
    Module.Write Mass=true
    Module.Write Material Random=true
    Module.Write Position=false
    Module.Write Scale=false
    Module.Write SpriteRotation=false
    Module.Write SpriteSize=true
    Position=&SimulationPosition()
    Sprite Size=<100 100>
   )]
  particleUpdateScript [
   &ParticleState()
   /SetVariables(Particles.Debug:NiagaraFloat=$System.InterpolateOverTime.InterpolateOverTime_Float.Moving Average) {isSetParameters=true}
   &DynamicMaterialParameters(
    Index 0 Param 1=$Particles.Debug
    Module.Param0WriteEnabled=true
    Module.Param1WriteEnabled=true
    Module.Param2WriteEnabled=true
    Module.Param3WriteEnabled=true
   )
   &SystemLocation(MaskSpawnGroup=false, Offset=<-50 0 170>)
  ]
  rendererClasses [&NiagaraSpriteRendererProperties]
  renderers [{rendererClass=&NiagaraSpriteRendererProperties}]
 }
 GlowTop {
  emitterUpdateScript [
   &EmitterState(Life Cycle Mode=::NewEnumerator1, Loop Behavior=::NewEnumerator1, Loop Duration Mode=::NewEnumerator1, UseLoopDelay=false)
   &SpawnBurst_Instantaneous(Spawn Count=1, Use Spawn Probability=false)
  ]
  particleSpawnScript [/Niagara/Modules/Spawn/Initialization/V2/InitializeParticle(
    Color=<rgba 0.5558 0.8714 0.0953 0.6>
    Sprite Size Mode=::NewEnumerator3
    Uniform Sprite Size=28
    UsePositionOffset=false
    Write Material Random=true
   ) {module=InitializeParticle001}]
  particleUpdateScript [
   &ParticleState(Kill Particles When Lifetime Has Elapsed=false)
   &CameraOffset(Camera Offset Amount=5)
   &SkeletalMeshLocation(
    Bone / Socket Index=1
    Bone Sampling Mode=::NewEnumerator4
    MaskSpawnGroup=false
    Sampled Position Offset=<0 0 118>
    Skeletal Mesh=$System.Skeletal Mesh
   )
   &SolveForcesAndVelocity(Clamp Velocity=false, Limit Acceleration=false)
   &ScaleColor(
    Scale Alpha=&Multiply_Float(
     A=$System.RevealCurve
     B=&Multiply_Float(
      A=&FNTimeOfDayScalar(Night Scale=0.7, Morning Scale=0.8, Evening Scale=0.8)
      B=&FN_FrontEndFloatModifier(Gameplay=1, Front End=0.5)
     )
    )
    ScaleA=true
    ScaleRGB=false
   )
  ]
  rendererClasses [&NiagaraSpriteRendererProperties]
  renderers [{rendererClass=&NiagaraSpriteRendererProperties}]
 }
 Trail_L {
  emitterSpawnScript [/SetVariables(Emitter.RibbonRandom:NiagaraFloat=&RandomRangeFloat(Minimum=0, Maximum=1)) {isSetParameters=true, module=SetVariables_#10}]
  emitterUpdateScript [
   &EmitterState(
    Inactive Response=::NewEnumerator1
    Life Cycle Mode=::NewEnumerator1
    Loop Behavior=::NewEnumerator1
    Loop Duration Mode=::NewEnumerator1
    UseLoopDelay=false
   )
   &Timeline(Play=&SetBoolByFloatComparison(A=$User.ForwardDot, B=0.1))
   &InterpolateOverTime(Target Float Value:NiagaraFloat=$User.RightDot)
   &InterpolateOverTime(Target Float Value:NiagaraFloat=$User.ForwardDot) {module=InterpolateOverTime001}
   /SetVariables(
    Emitter.PositionOffset:Vector3f=&FN_FrontEndVector3Modifier(Gameplay=<3 0 120>, Front End=<3 0 115>)
    Emitter.RibbonLength:NiagaraFloat=&Lerp_Float(A=100, B=100, Alpha=$StackContext.Timeline.Time)
   ) {isSetParameters=true, module=SetVariables_#11}
   &SpawnBurst_Instantaneous(Spawn Count=14, Use Spawn Probability=false, UseLoopCountLimit=false)
   /BRCosmetics/Animation/Game/MainPlayer/Emotes/UpbeatIguana/FX/InterpolateTrailVector(Range=10, Target Vector:Vector3f=$Engine.Owner.SystemYAxis) {module=InterpolateTrailVector001}
   /SetVariables(Emitter.InterpolatedVectorYAxis:Vector3f=$StackContext.InterpolatedVector, Emitter.TargetVectorYAxis:Vector3f=$StackContext.TargetVector) {isSetParameters=true, module=SetVariables_#12}
   &InterpolateTrailVector(Range=10, Target Vector:Vector3f=$Engine.Owner.SystemZAxis)
  ]
  particleSpawnScript [
   &InitializeParticle(Color Mode=::NewEnumerator0, Sprite Size Mode=::NewEnumerator3, UsePositionOffset=false, Write Material Random=true)
   /SetVariables(Particles.NormalizedRibbonPos:NiagaraFloat=&ReturnNormalizedExecIndex()) {isSetParameters=true, module=SetVariables_#13}
   /SetVariables(Particles.RibbonWidth=&Multiply_Float(
     A=13
     B=&Lerp_Float(
      A=2.5
      B=0.2
      Alpha=&FloatFromCurve(
       FloatCurve=DataInterface() {propertyValues {Curve {keys [{interpMode=RCIM_Cubic, tangentMode=RCTM_Break, time=0, value=1}, {interpMode=RCIM_Cubic, tangentMode=RCTM_Break, time=0.6, value=0}]}}}
       CurveIndex=&OneMinusFloat(Float=$Particles.NormalizedRibbonPos)
      )
     )
    )) {isSetParameters=true, module=SetVariables_#14}
   /SetVariables(Particles.UnderlayRibbonWidth:NiagaraFloat=&Multiply_Float(A=$Particles.RibbonWidth, B=1.1)) {isSetParameters=true}
   /SetVariables(
    Particles.SpiralFreq=1.2
    Particles.SpiralStrength:NiagaraFloat=&Multiply_Float(
     A=18
     B=&FloatFromCurve(
      FloatCurve=DataInterface() {propertyValues {Curve {keys [
          {arriveTangent=3.6599, interpMode=RCIM_Cubic, time=0.05, value=0}
          <at 0.5 1 RCIM_Cubic>
          {interpMode=RCIM_Cubic, leaveTangent=-0.1348, time=0.95, value=0}
         ]}}}
      CurveIndex=$Particles.NormalizedRibbonPos
     )
    )
   ) {isSetParameters=true, module=SetVariables_#15}
  ]
  particleUpdateScript [
   /SetVariables(Particles.PositionBend:Vector3f=&Multiply_VectorByFloat(
     Vector=&TransformVector(
      Vector=&MakeVector(
       X=&Multiply_Float(
        A=90
        B=&Lerp_Float(
         A=1
         B=&ClampFloat(Float=$System.InterpolateOverTime.InterpolateOverTime_Float.Moving Average, Min=-0.5, Max=0.7)
         Alpha=&Abs_Float(Float:NiagaraFloat=$System.InterpolateOverTime.InterpolateOverTime_Float.Moving Average)
        )
       )
       Y=0
       Z=0
      )
      Destination Space=::Simulation
     )
     Float=&Power(a=&OneMinusFloat(Float=&ReturnNormalizedExecIndex()), b=2)
    )) {isSetParameters=true, module=SetVariables_#16}
   /SetVariables(Particles.SpiralStrength:NiagaraFloat=&Multiply_Float(
     A=20
     B=&Multiply_Float(
      A=&FloatFromCurve(
       FloatCurve=DataInterface() {propertyValues {Curve {keys [
           {arriveTangent=5.9473, interpMode=RCIM_Cubic, time=0.1, value=0.3}
           {arriveTangent=-0.375, interpMode=RCIM_Cubic, leaveTangent=-0.375, time=0.5, value=1}
           {interpMode=RCIM_Cubic, leaveTangent=-2.9879, time=0.9, value=0}
          ]}}}
       CurveIndex=$Particles.NormalizedRibbonPos
      )
      B=$System.DotCameraZAxis
     )
    )) {isSetParameters=true, module=SetVariables_#17}
   /SetVariables(Particles.SpiralTime:NiagaraFloat=&Add_Float(
     A=&Add_Float(A=&Multiply_Float(A=$System.Age, B=0.5), B=$Particles.NormalizedRibbonPos)
     B=&Multiply_FloatByInt(
      Float=&Divide_Float(A=$Particles.SpiralFreq, B=&MakeFloatFromInt(Int=1))
      Integer=1
     )
    )) {isSetParameters=true, module=SetVariables_#18}
   /SetVariables(Particles.SpiralOffset:Vector3f=&MakeVector(
     X=&Sine(Normalized Angle=$Particles.SpiralTime, Period=$Particles.SpiralFreq, Scale=$Particles.SpiralStrength, Bias=0)
     Y=&Cosine(Normalized Angle=$Particles.SpiralTime, Period=$Particles.SpiralFreq, Scale=$Particles.SpiralStrength, Bias=0)
     Z=0
    )) {isSetParameters=true, module=SetVariables_#19}
   /SetVariables(
    Particles.InterpolatedTrailVector:Vector3f=&TransformVector(
     Vector=/Niagara/DynamicInputs/Math/SlerpVector(
      Vector A:Vector3f=$Emitter.InterpolatedVector
      Vector B:Vector3f=$Emitter.TargetVector
      Alpha:NiagaraFloat=&ReturnNormalizedExecIndex()
      Scale:NiagaraFloat=1
     )
     Source Space=::World
     Destination Space=::Simulation
     Apply Scale=true
    )
    Particles.InterpolatedTrailVectorYAxis:Vector3f=&TransformVector(
     Vector=&SlerpVector(
      Vector A:Vector3f=$Emitter.InterpolatedVectorYAxis
      Vector B:Vector3f=$Emitter.TargetVectorYAxis
      Alpha:NiagaraFloat=&ReturnNormalizedExecIndex()
      Scale:NiagaraFloat=1
     )
     Source Space=::World
     Destination Space=::Simulation
     Apply Scale=true
    )
    Particles.RibbonFacing=&TransformVector(Vector=$Engine.Owner.SystemYAxis, Source Space=::World, Destination Space=::Simulation)
   ) {isSetParameters=true, module=SetVariables_#20}
   /SetVariables(
    Particles.BackOffsetVector:Vector3f=&Multiply_VectorByFloat(
     Vector=&NormalizeVector(
      Vector To Normalize=/Niagara/DynamicInputs/Vectors/CrossProduct(
       Vector A:Vector3f=$StackContext.InterpolatedTrailVector
       Vector B:Vector3f=&TransformVector(Vector=$Engine.Owner.SystemXAxis, Source Space=::World, Destination Space=::Simulation)
       Normalize Incoming Vectors:NiagaraBool=true
      )
      Safe Normalize=true
      Safe Fallback Threshold=0
      Safe Fallback Vector=<0 0 0.001>
     )
     Float=&Multiply_Float(
      A=$Emitter.InterpolateOverTime001.InterpolateOverTime_Float.Moving Average
      B=&Multiply_Float(
       A=&Power(a=&OneMinusFloat(Float=&ReturnNormalizedExecIndex()), b=3)
       B=-99
      )
     )
    )
    Particles.SidewaysOffsetVector:Vector3f=&Multiply_VectorByFloat(
     Vector=&NormalizeVector(
      Vector To Normalize=&CrossProduct(
       Vector A:Vector3f=$StackContext.InterpolatedTrailVector
       Vector B:Vector3f=$Particles.RibbonFacing
       Normalize Incoming Vectors:NiagaraBool=true
      )
      Safe Normalize=true
      Safe Fallback Threshold=0
      Safe Fallback Vector=<0 0 0.001>
     )
     Float=&Multiply_Float(
      A=$Emitter.InterpolateOverTime.InterpolateOverTime_Float.Moving Average
      B=&Multiply_Float(
       A=&Power(a=&OneMinusFloat(Float=&ReturnNormalizedExecIndex()), b=2)
       B=-99
      )
     )
    )
   ) {isSetParameters=true, module=SetVariables_#21}
   /SetVariables(Particles.InterpolatedPosition:NiagaraPosition=&AddVectorToPosition(
     Position=&AddVectorToPosition(
      Position=$Engine.Emitter.SimulationPosition
      Vector=&Add_Vector(
       A=&Add_Vector(A=$Emitter.PositionOffset, B=&Add_Vector(A=$Particles.PositionBend, B=$Particles.SpiralOffset))
       B=&Add_Vector(A=$StackContext.SidewaysOffsetVector, B=$StackContext.BackOffsetVector)
      )
     )
     Vector=&Multiply_VectorByFloat(
      Vector=&RotateVectorByQuaternion(
       Vector To Rotate=$Particles.InterpolatedTrailVectorYAxis
       Quaternion=&MakeQuatFromAxisAngle(Axis=<1 0 0>, AngleInDegrees=90)
      )
      Float=&Multiply_Float(A=&OneMinusFloat(Float=&ReturnNormalizedExecIndex()), B=$Emitter.RibbonLength)
     )
    )) {isSetParameters=true}
   &DynamicMaterialParameters(
    Index 0 Param 1=&Subtract_Float(A=&Multiply_Float(A=$System.FN_Integrator002.Value, B=-0.7), B=$Emitter.RibbonRandom)
    Index 0 Param 4=$System.RevealCurve
    Module.Index0Param0Write=true
    Module.Index0Param1Write=true
    Module.Index0Param2Write=true
    Module.Index0Param3Write=true
    Module.Index1Param0Write=true
    Module.Index1Param1Write=true
    Module.Index1Param2Write=true
    Module.Index1Param3Write=true
    Write Parameter Index 1=true
   )
   &SampleSkeletalMesh(
    AdditionalYawPitchRoll=false
    Bone Sampling Mode:ENiagaraBoneSamplingMode=::NewEnumerator4
    Skeletal Mesh:NiagaraDataInterfaceSkeletalMesh=$System.Skeletal Mesh
   )
   /SetVariables(Particles.InterpolatedPosition:NiagaraPosition=&FN_FrontEndVector3Modifier(
     Gameplay=Linked() {linkedVariable {name=StackContext.InterpolatedPosition, type=NiagaraPosition}}
     Front End=&AddVectorToPosition(
      Position=&RotateVectorByQuaternion(
       Vector To Rotate=Linked() {linkedVariable {name=StackContext.InterpolatedPosition, type=NiagaraPosition}}
       Quaternion=$Particles.SampleSkeletalMesh.SampledRotation
      )
      Vector=$Particles.SampleSkeletalMesh.SampledPosition
     )
    )) {isSetParameters=true}
   /SetVariables(Particles.Position=$Particles.InterpolatedPosition) {isSetParameters=true, module=SetVariables_#22}
  ]
  rendererClasses [&NiagaraRibbonRendererProperties]
  renderers [
   {rendererClass=&NiagaraRibbonRendererProperties}
   {rendererClass=&NiagaraRibbonRendererProperties, rendererIndex=1}
  ]
 }
 Trail_R {
  emitterSpawnScript [/SetVariables(Emitter.RibbonRandom:NiagaraFloat=&RandomRangeFloat(Minimum=0, Maximum=1)) {isSetParameters=true, module=SetVariables_#10}]
  emitterUpdateScript [
   &EmitterState(
    Inactive Response=::NewEnumerator1
    Life Cycle Mode=::NewEnumerator1
    Loop Behavior=::NewEnumerator1
    Loop Duration Mode=::NewEnumerator1
    UseLoopDelay=false
   )
   &SampleSkeletalMesh(
    AdditionalYawPitchRoll=false
    Bone Sampling Mode:ENiagaraBoneSamplingMode=::NewEnumerator4
    Skeletal Mesh:NiagaraDataInterfaceSkeletalMesh=$System.Skeletal Mesh
   )
   &Timeline(Play=&SetBoolByFloatComparison(A=$User.ForwardDot, B=0.1))
   &InterpolateOverTime(Target Float Value:NiagaraFloat=$User.RightDot)
   &InterpolateOverTime(Target Float Value:NiagaraFloat=$User.ForwardDot) {module=InterpolateOverTime001}
   /SetVariables(
    Emitter.PositionOffset:Vector3f=&FN_FrontEndVector3Modifier(Gameplay=<-2 0 120>, Front End=<-2 0 115>)
    Emitter.RibbonLength:NiagaraFloat=&Lerp_Float(A=60, B=60, Alpha=$StackContext.Timeline.Time)
   ) {isSetParameters=true, module=SetVariables_#11}
   &SpawnBurst_Instantaneous(Spawn Count=10, Use Spawn Probability=false, UseLoopCountLimit=false)
   &InterpolateTrailVector(Range=10, Target Vector:Vector3f=$Engine.Owner.SystemYAxis) {module=InterpolateTrailVector001}
   /SetVariables(Emitter.InterpolatedVectorYAxis:Vector3f=$StackContext.InterpolatedVector, Emitter.TargetVectorYAxis:Vector3f=$StackContext.TargetVector) {isSetParameters=true, module=SetVariables_#12}
   &InterpolateTrailVector(Range=0.3, Target Vector:Vector3f=$Engine.Owner.SystemZAxis)
  ]
  particleSpawnScript [
   &InitializeParticle(Sprite Size Mode=::NewEnumerator3, UsePositionOffset=false, Write Material Random=true)
   /SetVariables(Particles.NormalizedRibbonPos:NiagaraFloat=&ReturnNormalizedExecIndex()) {isSetParameters=true, module=SetVariables_#13}
   /SetVariables(Particles.RibbonWidth=&Multiply_Float(
     A=9
     B=&Lerp_Float(
      A=2.5
      B=0.2
      Alpha=&FloatFromCurve(
       FloatCurve=DataInterface() {propertyValues {Curve {keys [{interpMode=RCIM_Cubic, tangentMode=RCTM_Break, time=0, value=1}, {interpMode=RCIM_Cubic, tangentMode=RCTM_Break, time=0.6, value=0}]}}}
       CurveIndex=&OneMinusFloat(Float=$Particles.NormalizedRibbonPos)
      )
     )
    )) {isSetParameters=true, module=SetVariables_#14}
   /SetVariables(Particles.UnderlayRibbonWidth:NiagaraFloat=&Multiply_Float(A=$Particles.RibbonWidth, B=1.1)) {isSetParameters=true}
   /SetVariables(
    Particles.SpiralFreq=3
    Particles.SpiralStrength:NiagaraFloat=&Multiply_Float(
     A=18
     B=&FloatFromCurve(
      FloatCurve=DataInterface() {propertyValues {Curve {keys [
          {arriveTangent=3.6599, interpMode=RCIM_Cubic, time=0.05, value=0}
          <at 0.5 1 RCIM_Cubic>
          {interpMode=RCIM_Cubic, leaveTangent=-0.1348, time=0.95, value=0}
         ]}}}
      CurveIndex=$Particles.NormalizedRibbonPos
     )
    )
    Particles.SpiralTime=0.3
   ) {isSetParameters=true, module=SetVariables_#15}
  ]
  particleUpdateScript [
   /SetVariables(Particles.PositionBend:Vector3f=&Multiply_VectorByFloat(
     Vector=&TransformVector(
      Vector=&MakeVector(
       X=&Multiply_Float(
        A=-40
        B=&Lerp_Float(
         A=1
         B=&ClampFloat(Float=$System.InterpolateOverTime.InterpolateOverTime_Float.Moving Average, Min=-0.5, Max=0)
         Alpha=&Abs_Float(Float:NiagaraFloat=$System.InterpolateOverTime.InterpolateOverTime_Float.Moving Average)
        )
       )
       Y=0
       Z=0
      )
      Destination Space=::Simulation
     )
     Float=&Power(a=&OneMinusFloat(Float=&ReturnNormalizedExecIndex()), b=2)
    )) {isSetParameters=true, module=SetVariables_#16}
   /SetVariables(Particles.SpiralStrength:NiagaraFloat=&Multiply_Float(
     A=20
     B=&Multiply_Float(
      A=&FloatFromCurve(
       FloatCurve=DataInterface() {propertyValues {Curve {keys [
           {arriveTangent=5.9473, interpMode=RCIM_Cubic, time=0.1, value=0.3}
           {arriveTangent=-0.375, interpMode=RCIM_Cubic, leaveTangent=-0.375, time=0.5, value=1}
           {interpMode=RCIM_Cubic, leaveTangent=-2.9879, time=0.9, value=0}
          ]}}}
       CurveIndex=$Particles.NormalizedRibbonPos
      )
      B=$System.DotCameraZAxis
     )
    )) {isSetParameters=true, module=SetVariables_#17}
   /SetVariables(Particles.SpiralTime:NiagaraFloat=&Add_Float(
     A=&Add_Float(A=&Multiply_Float(A=$System.Age, B=0.5), B=$Particles.NormalizedRibbonPos)
     B=&Multiply_FloatByInt(
      Float=&Divide_Float(A=$Particles.SpiralFreq, B=&MakeFloatFromInt(Int=1))
      Integer=1
     )
    )) {isSetParameters=true, module=SetVariables_#18}
   /SetVariables(Particles.SpiralOffset:Vector3f=&MakeVector(
     X=&Sine(Normalized Angle=$Particles.SpiralTime, Period=$Particles.SpiralFreq, Scale=$Particles.SpiralStrength, Bias=0)
     Y=&Cosine(Normalized Angle=$Particles.SpiralTime, Period=$Particles.SpiralFreq, Scale=$Particles.SpiralStrength, Bias=0)
     Z=0
    )) {isSetParameters=true, module=SetVariables_#19}
   /SetVariables(
    Particles.InterpolatedTrailVector:Vector3f=&TransformVector(
     Vector=&SlerpVector(
      Vector A:Vector3f=$Emitter.InterpolatedVector
      Vector B:Vector3f=$Emitter.TargetVector
      Alpha:NiagaraFloat=&ReturnNormalizedExecIndex()
      Scale:NiagaraFloat=1
     )
     Source Space=::World
     Destination Space=::Simulation
     Apply Scale=true
    )
    Particles.InterpolatedTrailVectorYAxis:Vector3f=&TransformVector(
     Vector=&SlerpVector(
      Vector A:Vector3f=$Emitter.InterpolatedVectorYAxis
      Vector B:Vector3f=$Emitter.TargetVectorYAxis
      Alpha:NiagaraFloat=&ReturnNormalizedExecIndex()
      Scale:NiagaraFloat=1
     )
     Source Space=::World
     Destination Space=::Simulation
     Apply Scale=true
    )
    Particles.RibbonFacing=&TransformVector(Vector=$Engine.Owner.SystemYAxis, Source Space=::World, Destination Space=::Simulation)
   ) {isSetParameters=true, module=SetVariables_#20}
   /SetVariables(
    Particles.BackOffsetVector:Vector3f=&Multiply_VectorByFloat(
     Vector=&NormalizeVector(
      Vector To Normalize=&CrossProduct(
       Vector A:Vector3f=$StackContext.InterpolatedTrailVector
       Vector B:Vector3f=&TransformVector(Vector=$Engine.Owner.SystemXAxis, Source Space=::World, Destination Space=::Simulation)
       Normalize Incoming Vectors:NiagaraBool=true
      )
      Safe Normalize=true
      Safe Fallback Threshold=0
      Safe Fallback Vector=<0 0 0.001>
     )
     Float=&Multiply_Float(
      A=$Emitter.InterpolateOverTime001.InterpolateOverTime_Float.Moving Average
      B=&Multiply_Float(
       A=&Power(a=&OneMinusFloat(Float=&ReturnNormalizedExecIndex()), b=3)
       B=-99
      )
     )
    )
    Particles.SidewaysOffsetVector:Vector3f=&Multiply_VectorByFloat(
     Vector=&NormalizeVector(
      Vector To Normalize=&CrossProduct(
       Vector A:Vector3f=$StackContext.InterpolatedTrailVector
       Vector B:Vector3f=$Particles.RibbonFacing
       Normalize Incoming Vectors:NiagaraBool=true
      )
      Safe Normalize=true
      Safe Fallback Threshold=0
      Safe Fallback Vector=<0 0 0.001>
     )
     Float=&Multiply_Float(
      A=$Emitter.InterpolateOverTime.InterpolateOverTime_Float.Moving Average
      B=&Multiply_Float(
       A=&Power(a=&OneMinusFloat(Float=&ReturnNormalizedExecIndex()), b=2)
       B=-133
      )
     )
    )
   ) {isSetParameters=true, module=SetVariables_#21}
   /SetVariables(Particles.InterpolatedPosition:NiagaraPosition=&AddVectorToPosition(
     Position=&AddVectorToPosition(
      Position=$Engine.Emitter.SimulationPosition
      Vector=&Add_Vector(
       A=&Add_Vector(A=$Emitter.PositionOffset, B=&Add_Vector(A=$Particles.PositionBend, B=$Particles.SpiralOffset))
       B=&Add_Vector(A=$StackContext.SidewaysOffsetVector, B=$StackContext.BackOffsetVector)
      )
     )
     Vector=&Multiply_VectorByFloat(
      Vector=&RotateVectorByQuaternion(
       Vector To Rotate=$Particles.InterpolatedTrailVectorYAxis
       Quaternion=&MakeQuatFromAxisAngle(Axis=<1 0 0>, AngleInDegrees=90)
      )
      Float=&Multiply_Float(A=&OneMinusFloat(Float=&ReturnNormalizedExecIndex()), B=$Emitter.RibbonLength)
     )
    )) {isSetParameters=true}
   &DynamicMaterialParameters(
    Index 0 Param 1=&Subtract_Float(A=&Multiply_Float(A=$System.FN_Integrator002.Value, B=-0.85), B=$Emitter.RibbonRandom)
    Index 0 Param 4=$System.RevealCurve
    Module.Index0Param0Write=true
    Module.Index0Param1Write=true
    Module.Index0Param2Write=true
    Module.Index0Param3Write=true
    Module.Index1Param0Write=true
    Module.Index1Param1Write=true
    Module.Index1Param2Write=true
    Module.Index1Param3Write=true
    Write Parameter Index 1=true
   )
   &SampleSkeletalMesh(
    AdditionalYawPitchRoll=false
    Bone Sampling Mode:ENiagaraBoneSamplingMode=::NewEnumerator4
    Skeletal Mesh:NiagaraDataInterfaceSkeletalMesh=$System.Skeletal Mesh
   ) {module=SampleSkeletalMesh001}
   /SetVariables(Particles.InterpolatedPosition:NiagaraPosition=&FN_FrontEndVector3Modifier(
     Gameplay=$StackContext.InterpolatedPosition
     Front End=&AddVectorToPosition(
      Position=&RotateVectorByQuaternion(Vector To Rotate=$StackContext.InterpolatedPosition, Quaternion=$Particles.SampleSkeletalMesh001.SampledRotation)
      Vector=$Particles.SampleSkeletalMesh001.SampledPosition
     )
    )) {isSetParameters=true}
   /SetVariables(Particles.Position=$Particles.InterpolatedPosition) {isSetParameters=true, module=SetVariables_#22}
  ]
  rendererClasses [&NiagaraRibbonRendererProperties]
  renderers [
   {rendererClass=&NiagaraRibbonRendererProperties}
   {rendererClass=&NiagaraRibbonRendererProperties, rendererIndex=1}
  ]
 }
}
summary {
 systemName=NS_Umbrella_Season_31_Trail
 emitters [
  {emitterName=Trail_L, rendererClasses [&NiagaraRibbonRendererProperties]}
  {emitterName=Trail_R, rendererClasses [&NiagaraRibbonRendererProperties]}
  {emitterName=Dots_L, rendererClasses [&NiagaraSpriteRendererProperties]}
  {emitterName=Dots_R, rendererClasses [&NiagaraSpriteRendererProperties]}
  {bEnabled=false, emitterName=FN_Debug, rendererClasses [&NiagaraSpriteRendererProperties]}
  {emitterName=GlowTop, rendererClasses [&NiagaraSpriteRendererProperties]}
 ]
 userVariables [
  User.bIsFrontEnd {type=NiagaraBool}
  User.bIsFrontEndPreview {type=NiagaraBool}
  User.bIsFullyDeployed {type=NiagaraBool}
  User.ForwardDot {type=NiagaraFloat, defaultValue=0, description=""}
  User.RightDot {type=NiagaraFloat, defaultValue=0, description=""}
  User.Variant {type=NiagaraFloat, defaultValue=0, description=""}
 ]
}
