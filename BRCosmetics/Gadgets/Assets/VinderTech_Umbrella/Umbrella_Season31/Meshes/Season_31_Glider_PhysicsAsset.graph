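# Physics asset for the Season 31 VinderTech umbrella glider mesh, giving it simplified collision shapes.
# One capsule body covers the handle bone (Melee_TwoHanded_Handle), and a much wider capsule, rotated about 90 degrees, covers blade_01, likely the canopy.
# A single constraint joins blade_01 to the handle with matching swing and twist limit angles.
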PhysicsAsset /BRCosmetics/Gadgets/Assets/VinderTech_Umbrella/Umbrella_Season31/Meshes/Season_31_Glider_PhysicsAsset
bodies {
 Melee_TwoHanded_Handle [{shapeType=Capsule, length=57.7483, radius=9.2228, center=<0.1286 -0.0716 38.4428>, rotation=<rot 0.0938 -90 -0.0262>}]
 blade_01 [{shapeType=Capsule, length=83.7433, radius=82.3559, center=<-0.0016 0.0022 -3.7315>, rotation=<rot 0 -17.9441 -89.995>}]
}
constraints [{bone1Name=blade_01, bone2Name=Melee_TwoHanded_Handle, swing1LimitDegrees=45, swing2LimitDegrees=45, twistLimitDegrees=45}]
